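; B_Pickups_Morphite is a Blueprint for a Morphite pickup in the BlastBerry weapons content.
; Its On Tossed event and construction script only call the parent implementations, so its graphs add no logic beyond the inherited behavior.
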(event EventOnTossed
  (|CallParentFunction))

(fn ConstructionScript
  (|CallParentFunction))
