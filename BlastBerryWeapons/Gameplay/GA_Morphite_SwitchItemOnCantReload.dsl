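; Gameplay ability that upgrades a chrome weapon with a pending power-up when the player presses reload but lacks ammo to reload.
; Input triggers TrySwitchWeapons, which on dedicated servers finds the tag-matched weapon's XP component and checks magazine count equals remaining ammo.
; The server and owning client, via Client_PlayUpgradeMontage, play an overheat montage, then flush the pending power-up on blend-out.
; Normal reloads leave flushing to GA_Chrome_Reload, per its comments.
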(event Ability|EventFailedToActivatePassiveAbility
  (params (ActorInfo :type Gameplay%20Ability%20Actor%20Info%20Structure))
  ;[0.229 1 0] We don't need to do anything when this ability activates, but declare these events to prevent unnecessary log messages.
  (|CallParentFunction))

(event Ability|EventActivateAbility
  ;[0.229 1 0] We don't need to do anything when this ability activates, but declare these events to prevent unnecessary log messages.
  (|CallParentFunction))

(event Ability|EventOnAbilityInputPressed
  ;[0.09912 1 0.06801] Tell the server that the player tried to reload.
  (|CallParentFunction)
  (CallFunction|TrySwitchWeapons))

(event Custom|TrySwitchWeapons
  (bind _itemxpcomponent (CallFunction|TryGetChromeXPComponent))
  (bind _itemxpcomponent_2 (CallFunction|TryGetChromeXPComponent))
  (if (Networking|IsDedicatedServer)
    ;[0.1232 0.3635 1] Try to get the chrome XP component of the current weapon and check if it has a pending power up. Chrome XP components only exist on chrome weapons so if the current weapon is not chrome, execution will stop here.
    (Utilities|IsValid _itemxpcomponent_2
      (:Is%20Valid
        (if (MultiItem|PowerUp|IsPowerUpPending _itemxpcomponent_2)
          ;[0.1875 1 1] Only try to upgrade if the player cannot reload due to low ammo. If the player has enough ammo to reload, the upgrade will be flushed in GA_Chrome_Reload after the reload is complete.
          (bind _outfortplayerpawn (CallFunction|GetAvatarasValidFortPlayerPawn)
            (:ValidFortPlayerPawn
              (Utilities|IsValid (Weapon|GetCurrentWeapon _outfortplayerpawn)
                (:Is%20Valid
                  (if (== (FortWeapon|GetMagazineAmmoCount (Weapon|GetCurrentWeapon _outfortplayerpawn)) (FortWeapon|GetRemainingAmmo (Weapon|GetCurrentWeapon _outfortplayerpawn)))
                    ;[0.7609 0.2012 1] Play the reload montage on the client and server and wait for it to be complete before actually flushing the pending power up.
                    ;;[1 0.9963 0.1317 1 12] Only replicate this to the owning client, playing the montage on the server will automatically multicast the montage to non-owning clients.
                    (CallFunction|ClientPlayUpgradeMontage)
                    (bind _outfortplayerpawn_2 (CallFunction|GetAvatarasValidFortPlayerPawn)
                      (:ValidFortPlayerPawn
                        (Ability|Tasks|PlayMontageAndWait "" "/Game/Animation/Game/MainPlayer/Combat/Guns/GalileoBun/GalileoBun_Overheat_CMM_M.GalileoBun_Overheat_CMM_M" 1.5
                          (:OnBlendOut
                            (Utilities|IsValid _itemxpcomponent
                              (:Is%20Valid
                                (MultiItem|PowerUp|FlushPendingPowerUp _itemxpcomponent))
                              (:Is%20Not%20Valid
                                _))
                            (bind _outfortplayerpawn_3 (CallFunction|GetAvatarasValidFortPlayerPawn))))))))))))))))

(event Custom|Client_PlayUpgradeMontage
  (bind _itemxpcomponent (CallFunction|TryGetChromeXPComponent))
  ;[0.7609 0.2012 1] Play the reload montage on the client and server and wait for it to be complete before actually flushing the pending power up.
  (bind _outfortplayerpawn (CallFunction|GetAvatarasValidFortPlayerPawn)
    (:ValidFortPlayerPawn
      (Ability|Tasks|PlayMontageAndWait "" "/Game/Animation/Game/MainPlayer/Combat/Guns/GalileoBun/GalileoBun_Overheat_CMM_M.GalileoBun_Overheat_CMM_M" 1.5
        (:OnBlendOut
          (Utilities|IsValid _itemxpcomponent
            (:Is%20Valid
              (MultiItem|PowerUp|FlushPendingPowerUp _itemxpcomponent))
            (:Is%20Not%20Valid
              _))
          (bind _outfortplayerpawn_2 (CallFunction|GetAvatarasValidFortPlayerPawn)))))))

(fn TryGetChromeXPComponent
  (bind _returnvalue (CallFunction|GetAvatarasFortPlayerPawn))
  (bind _returnvalue_2 (Weapon|GetCurrentWeapon _returnvalue))
  (Utilities|IsValid _returnvalue
    (:Is%20Valid
      (Utilities|IsValid _returnvalue_2
        (:Is%20Valid
          (if (Tags|WeaponDataHasAnyMatchingGameplayTags _returnvalue_2 (Variables|Default|GetChromeWeaponTags))
            (return (Actor|GetComponentbyClass _returnvalue_2 "/BlastBerryWeapons/Gameplay/C_RiftedWeapons_XPComponent.C_RiftedWeapons_XPComponent_C"))
            (else
              _)))
        (:Is%20Not%20Valid
          _))
      (return 0))
    (:Is%20Not%20Valid
      _)))
