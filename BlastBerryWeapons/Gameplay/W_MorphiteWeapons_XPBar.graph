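# HUD widget showing the current weapon's XP as a small progress bar, with a "Reload To Upgrade!" prompt at full XP.
# On Construct it binds to the owning pawn's current weapon's C_RiftedWeapons_XPComponent and immediately applies its XP percentage.
# Below 100%, updates set the bar, flash IncreaseXP and stop MaxXP.
# At 100%, MaxXP loops endlessly, driving the marquee and prompt visibility.
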WidgetBlueprint /BlastBerryWeapons/Gameplay/W_MorphiteWeapons_XPBar
animations [
 IncreaseXP {
  binding_count=1
  length_seconds=0.2807
  object=&self:IncreaseXP
  sequence {bindings [WeaponXPProgressBar {
     id=#1
     possessed_class=ProgressBar
     tracks [{
       type=MovieSceneColorTrack
       display_name=Fill Color and Opacity
       sections [{
         type=MovieSceneColorSection
         end_frame=5
         start_frame=-1
         channels [
          Color.R {type=MovieSceneScriptingFloatChannel, keys [[0,0],[5,1]]}
          Color.G {type=MovieSceneScriptingFloatChannel, const [1,2]}
          Color.B {type=MovieSceneScriptingFloatChannel, keys [[0,0.99],[5,1]]}
          Color.A {type=MovieSceneScriptingFloatChannel, keys [[0,1],[5,0.7]]}
         ]
        }]
      }]
    }]}
 }
 MaxXP {
  binding_count=2
  length_seconds=1
  object=&self:MaxXP
  sequence {bindings [
    WeaponXPProgressBar {
     id=#1
     possessed_class=ProgressBar
     tracks [
      {
       type=MovieSceneColorTrack
       display_name=Fill Color and Opacity
       sections [{
         type=MovieSceneColorSection
         end_frame=20
         start_frame=0
         channels [
          Color.R {type=MovieSceneScriptingFloatChannel, const [0,1]}
          Color.G {type=MovieSceneScriptingFloatChannel, const [1,1]}
          Color.B {type=MovieSceneScriptingFloatChannel, const [0.99,1]}
          Color.A {type=MovieSceneScriptingFloatChannel, const [1,1]}
         ]
        }]
      }
      {
       type=MovieSceneFloatTrack
       display_name=Percent
       sections [{type=MovieSceneFloatSection, end_frame=20, start_frame=0, channels [None {type=MovieSceneScriptingFloatChannel, keys [[-1,1],[0,0]]}]}]
      }
      {
       type=MovieSceneBoolTrack
       display_name=Is Marquee
       sections [{type=MovieSceneBoolSection, end_frame=20, start_frame=0, channels [None {type=MovieSceneScriptingBoolChannel, const [true,1]}]}]
      }
     ]
    }
    UpgradeText {
     id=#2
     possessed_class=CommonTextBlock
     tracks [{
       type=MovieSceneByteTrack
       display_name=Visibility
       sections [{type=MovieSceneByteSection, end_frame=10, start_frame=0, channels [None {type=MovieSceneScriptingByteChannel, keys [[0,4],[10,2]]}]}]
      }]
    }
   ]}
 }
]
graphs [EventGraph {nodes [
   n1 = OnXPChanged_Event!() -> (OutputDelegate, then, XPDelta:double=~, CurrentXPPercentage:double=~) {is=title K2Node_CustomEvent}
   > n2 = Branch(Condition="Not Equal (Float)"(A=@n1.CurrentXPPercentage, B=1))
   > SetPercent!(
    execute
    self:Progress Bar Object Reference=Get WeaponXPProgressBar() -> (WeaponXPProgressBar:Progress Bar Object Reference)
    InPercent:float=@n1.CurrentXPPercentage
   ) -> (then)
   > PlayAnimation(InAnimation=Get IncreaseXP() -> (IncreaseXP:Widget Animation Object Reference))
   > Branch(Condition=IsAnimationPlaying(InAnimation=Get MaxXP() -> (MaxXP:Widget Animation Object Reference)))
   > StopAnimation(InAnimation=Get MaxXP())
   PlayAnimation(execute=@n2.else, InAnimation=Get MaxXP(), NumLoopsToPlay=0)
   Event Construct()
   > n3 = Cast To FortPlayerPawnAthena(Object=GetOwningPlayerPawn!(self:User Widget Object Reference) -> (ReturnValue:Pawn Object Reference))
   n4 = GetCurrentWeapon!(self:Fort Pawn Object Reference=@n3) -> (ReturnValue:Fort Weapon Object Reference)
   n5 = GetComponentByClass!(
    self:Actor Object Reference=@n4
    ComponentClass:Actor Component Class Reference=&/C_RiftedWeapons_XPComponent.C_RiftedWeapons_XPComponent_C
   ) -> (ReturnValue:C Rifted Weapons XPComponent Object Reference)
   > Branch(Condition=AND Boolean(A=IsValid(Object=@n4), B=IsValid(Object=@n5)))
   > Bind Event to On XPChanged(self:C Rifted Weapons XPComponent Object Reference=@n5, Delegate=@n1) {is=title K2Node_AddDelegate}
   > OnXPChanged_Event!(
    execute
    self
    XPDelta:double=0
    CurrentXPPercentage:double=GetMultiItemXPPercentage!(self:Fort World Multi Item XPComponent Object Reference=@n5) -> (OutItemXPPercentage:float, ReturnValue:bool)
   ) -> (then) {is=title K2Node_CallFunction}
  ]}]
variables []
widgets {
 info {inheritedWidgetCount=0, parentClass=&UserWidget, rootWidgetClass=&Overlay, widgetCount=5}
 widgets [
  Overlay_184 {widgetClassPath=&Overlay, widget=Overlay_184}
  VerticalBox_38 {
   widgetClassPath=&VerticalBox
   parent=Overlay_184
   widget=VerticalBox_38
   slot {class=OverlaySlot, properties {horizontalAlignment=HAlign_Center, verticalAlignment=VAlign_Bottom, padding=<ltrb 0 0 0 385>}}
  }
  SizeBox {
   widgetClassPath=&SizeBox
   parent=VerticalBox_38
   widget=SizeBox
   properties {bOverride_HeightOverride=true, bOverride_WidthOverride=true, heightOverride=7, widthOverride=100}
   slot {class=VerticalBoxSlot, properties {horizontalAlignment=HAlign_Center, verticalAlignment=VAlign_Top}}
  }
  WeaponXPProgressBar {
   widgetClassPath=/Script/UMG.ProgressBar
   bIsVariable=true
   parent=SizeBox
   widget=WeaponXPProgressBar
   properties {
    fillColorAndOpacity=<rgba 1 1 1 0.7>
    widgetStyle {
     enableFillAnimation=false
     backgroundImage {
      bIsDynamicallyLoaded=false
      drawAs=Box
      imageType=FullColor
      mirroring=NoMirror
      resourceName=../../../Engine/Content/Slate/Common/ProgressBar_Background.png
      resourceObject=null
      tiling=NoTile
      imageSize=<0 0>
      margin=<ltrb 0.4167 0.4167 0.4167 0.4167>
      outlineSettings {
       bUseBrushTransparency=false
       roundingType=HalfHeightRadius
       width=0
       color {colorUseRule=UseColor_Specified, specifiedColor=<rgba 0 0 0 0>}
       cornerRadii=<0 0 0 0>
      }
      tintColor {colorUseRule=UseColor_Specified, specifiedColor=<rgba 0 0 0 0.3>}
      uVRegion {bIsValid=false, max=<0 0>, min=<0 0>}
     }
     fillImage {
      bIsDynamicallyLoaded=false
      drawAs=Box
      imageType=FullColor
      mirroring=NoMirror
      resourceName=../../../Engine/Content/Slate/Common/ProgressBar_Fill.png
      resourceObject=null
      tiling=NoTile
      imageSize=<0 0>
      margin=<ltrb 0.4167 0.4167 0.4167 0.4167>
      outlineSettings {
       bUseBrushTransparency=false
       roundingType=HalfHeightRadius
       width=0
       color {colorUseRule=UseColor_Specified, specifiedColor=<rgba 0 0 0 0>}
       cornerRadii=<0 0 0 0>
      }
      tintColor {colorUseRule=UseColor_Specified, specifiedColor=<rgba 1 1 1 1>}
      uVRegion {bIsValid=false, max=<0 0>, min=<0 0>}
     }
     marqueeImage {
      bIsDynamicallyLoaded=false
      drawAs=Image
      imageType=FullColor
      mirroring=NoMirror
      resourceName=../../../Engine/Content/Slate/Common/ProgressBar_Marquee.png
      resourceObject=null
      tiling=Horizontal
      imageSize=<20 12>
      margin=<ltrb 0 0 0 0>
      outlineSettings {
       bUseBrushTransparency=false
       roundingType=HalfHeightRadius
       width=0
       color {colorUseRule=UseColor_Specified, specifiedColor=<rgba 0 0 0 0>}
       cornerRadii=<0 0 0 0>
      }
      tintColor {colorUseRule=UseColor_Specified, specifiedColor=<rgba 0 1 1 1>}
      uVRegion {bIsValid=false, max=<0 0>, min=<0 0>}
     }
    }
   }
   slot {class=SizeBoxSlot, properties {verticalAlignment=VAlign_Top}}
  }
  UpgradeText {
   widgetClassPath=&CommonTextBlock
   bIsVariable=true
   parent=VerticalBox_38
   widget=UpgradeText
   properties {
    justification=Center
    text="Reload To Upgrade!"
    textTransformPolicy=ToUpper
    visibility=Hidden
    colorAndOpacity {colorUseRule=UseColor_Specified, specifiedColor=<rgba 0 1 0.9054 1>}
    font {
     bForceMonospaced=false
     bMaterialIsStencil=false
     fontMaterial=null
     fontObject=&Burbank
     letterSpacing=0
     monospacedWidth=1
     size=14
     skewAmount=0
     typefaceFontName=BurbankRegular-Black
     outlineSettings {
      bApplyOutlineToDropShadows=false
      bMiteredCorners=false
      bSeparateFillAlpha=false
      outlineMaterial=null
      outlineSize=1
      outlineColor=<rgba 0 0 0 1>
     }
    }
    renderTransform {angle=0, scale=<1 1>, shear=<-12 0>, translation=<0 0>}
    shadowOffset=<0 0>
   }
   slot {class=VerticalBoxSlot, properties {horizontalAlignment=HAlign_Center, verticalAlignment=VAlign_Top, padding=<ltrb 0 10 0 0>}}
  }
 ]
}
